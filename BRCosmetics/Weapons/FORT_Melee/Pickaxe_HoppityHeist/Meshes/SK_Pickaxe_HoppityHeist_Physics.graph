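# Physics asset for the HoppityHeist pickaxe skeletal mesh, defining collision and joints for its dangling ribbons and bows.
# The handle body uses three capsules, and every dynamic bone uses a box.
# Front and back five-bone ribbon chains hang from the handle with limited swing and slight twist.
# Four two-bone bow chains, two per side, attach to the handle with tighter swing and locked twist.
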PhysicsAsset /BRCosmetics/Weapons/FORT_Melee/Pickaxe_HoppityHeist/Meshes/SK_Pickaxe_HoppityHeist_Physics
bodies {
 Melee_TwoHanded_Handle [
  {shapeType=Capsule, length=73.029, radius=1.5719, center=<0 -1.4746 7.9818>}
  {shapeType=Capsule, length=11.4147, radius=1.5616, center=<-1.2645 -4.6557 -36.2584>, rotation=<rot -9.0547 -4.262 30.3367>}
  {shapeType=Capsule, length=11.6647, radius=1.5616, center=<1.5204 -4.7924 -36.2888>, rotation=<rot 9.0547 4.262 30.3366>}
 ]
 dyn_bck_ribbon_01 [{shapeType=Box, extentX=5.1206, extentY=0.8, extentZ=3.943, center=<1.1249 0 0>}]
 dyn_bck_ribbon_02 [{shapeType=Box, extentX=5.9391, extentY=0.8, extentZ=3.943, center=<2.3661 0 -0.0001>}]
 dyn_bck_ribbon_03 [{shapeType=Box, extentX=9.4747, extentY=0.8, extentZ=3.943, center=<2.7914 0 0.0001>}]
 dyn_bck_ribbon_04 [{shapeType=Box, extentX=7.677, extentY=0.8, extentZ=3.943, center=<2.847 0 0>}]
 dyn_bck_ribbon_05 [{shapeType=Box, extentX=6.9552, extentY=0.8, extentZ=3.943, center=<1.9642 0 0>}]
 dyn_fnt_ribbon_01 [{shapeType=Box, extentX=4.8238, extentY=0.8, extentZ=3.5608, center=<1.3213 0 0>}]
 dyn_fnt_ribbon_02 [{shapeType=Box, extentX=5.9222, extentY=0.8, extentZ=3.6173, center=<2.3178 0 0>}]
 dyn_fnt_ribbon_03 [{shapeType=Box, extentX=9.4548, extentY=0.8, extentZ=3.6173, center=<2.7508 0 0>}]
 dyn_fnt_ribbon_04 [{shapeType=Box, extentX=7.677, extentY=0.8, extentZ=3.6173, center=<2.9257 0 0>}]
 dyn_fnt_ribbon_05 [{shapeType=Box, extentX=6.9552, extentY=0.8, extentZ=3.6173, center=<2.0429 0 0>}]
 dyn_l_bow_a01 [{shapeType=Box, extentX=3.3466, extentY=5.082, extentZ=2.0118, center=<0.8347 0 -0.2776>}]
 dyn_l_bow_a02 [{shapeType=Box, extentX=4.9088, extentY=10.2501, extentZ=3.7598, center=<0.9215 0 0>}]
 dyn_l_bow_b01 [{shapeType=Box, extentX=3.3466, extentY=5.082, extentZ=2.0118, center=<0.8347 0 0.4311>}]
 dyn_l_bow_b02 [{shapeType=Box, extentX=3.6153, extentY=7.8996, extentZ=3.0181, center=<0.8947 0 0>}]
 dyn_r_bow_a01 [{shapeType=Box, extentX=3.3466, extentY=5.082, extentZ=2.0118, center=<1.2211 0 0.3599>}]
 dyn_r_bow_a02 [{shapeType=Box, extentX=4.9088, extentY=10.0235, extentZ=3.7598, center=<1.1214 0 0.0001>}]
 dyn_r_bow_b01 [{shapeType=Box, extentX=3.3466, extentY=5.082, extentZ=2.0118, center=<1.2211 0 -0.52>}]
 dyn_r_bow_b02 [{shapeType=Box, extentX=3.5342, extentY=7.8609, extentZ=2.937, center=<0.797 0 0>}]
}
constraints [
 {bone1Name=dyn_fnt_ribbon_01, bone2Name=Melee_TwoHanded_Handle, swing1LimitDegrees=15, swing2LimitDegrees=30, twistLimitDegrees=5}
 {bone1Name=dyn_fnt_ribbon_02, bone2Name=dyn_fnt_ribbon_01, swing1LimitDegrees=15, swing2LimitDegrees=30, twistLimitDegrees=5}
 {bone1Name=dyn_fnt_ribbon_03, bone2Name=dyn_fnt_ribbon_02, swing1LimitDegrees=15, swing2LimitDegrees=30, twistLimitDegrees=5}
 {bone1Name=dyn_fnt_ribbon_04, bone2Name=dyn_fnt_ribbon_03, swing1LimitDegrees=15, swing2LimitDegrees=30, twistLimitDegrees=5}
 {bone1Name=dyn_fnt_ribbon_05, bone2Name=dyn_fnt_ribbon_04, swing1LimitDegrees=15, swing2LimitDegrees=30, twistLimitDegrees=5}
 {bone1Name=dyn_bck_ribbon_01, bone2Name=Melee_TwoHanded_Handle, swing1LimitDegrees=15, swing2LimitDegrees=30, twistLimitDegrees=5}
 {bone1Name=dyn_bck_ribbon_02, bone2Name=dyn_bck_ribbon_01, swing1LimitDegrees=15, swing2LimitDegrees=30, twistLimitDegrees=5}
 {bone1Name=dyn_bck_ribbon_03, bone2Name=dyn_bck_ribbon_02, swing1LimitDegrees=15, swing2LimitDegrees=30, twistLimitDegrees=5}
 {bone1Name=dyn_bck_ribbon_04, bone2Name=dyn_bck_ribbon_03, swing1LimitDegrees=15, swing2LimitDegrees=30, twistLimitDegrees=5}
 {bone1Name=dyn_bck_ribbon_05, bone2Name=dyn_bck_ribbon_04, swing1LimitDegrees=15, swing2LimitDegrees=30, twistLimitDegrees=5}
 {
  bone1Name=dyn_l_bow_b01
  bone2Name=Melee_TwoHanded_Handle
  swing1LimitDegrees=10
  swing2LimitDegrees=10
  twistLimitDegrees=45
  twistMotion=Locked
 }
 {bone1Name=dyn_l_bow_b02, bone2Name=dyn_l_bow_b01, swing1LimitDegrees=10, swing2LimitDegrees=10, twistLimitDegrees=45, twistMotion=Locked}
 {
  bone1Name=dyn_l_bow_a01
  bone2Name=Melee_TwoHanded_Handle
  swing1LimitDegrees=10
  swing2LimitDegrees=10
  twistLimitDegrees=45
  twistMotion=Locked
 }
 {bone1Name=dyn_l_bow_a02, bone2Name=dyn_l_bow_a01, swing1LimitDegrees=10, swing2LimitDegrees=10, twistLimitDegrees=45, twistMotion=Locked}
 {
  bone1Name=dyn_r_bow_b01
  bone2Name=Melee_TwoHanded_Handle
  swing1LimitDegrees=10
  swing2LimitDegrees=10
  twistLimitDegrees=45
  twistMotion=Locked
 }
 {bone1Name=dyn_r_bow_b02, bone2Name=dyn_r_bow_b01, swing1LimitDegrees=10, swing2LimitDegrees=10, twistLimitDegrees=45, twistMotion=Locked}
 {
  bone1Name=dyn_r_bow_a01
  bone2Name=Melee_TwoHanded_Handle
  swing1LimitDegrees=10
  swing2LimitDegrees=10
  twistLimitDegrees=45
  twistMotion=Locked
 }
 {bone1Name=dyn_r_bow_a02, bone2Name=dyn_r_bow_a01, swing1LimitDegrees=10, swing2LimitDegrees=10, twistLimitDegrees=45, twistMotion=Locked}
]
