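# Animation Blueprint for the HoppityHeist pickaxe cosmetic that adds secondary motion to its front and back ribbons.
# The cached input pose feeds a RigidBody simulation with custom gravity, used when bRigidBodyEnabled is true.
# Otherwise, Trail controllers drive both ribbons over additive bone offsets chosen by bIsInFrontEnd: mid-ribbon bends in the front end, root-bone adjustments elsewhere.
# The event graph holds only unconnected stub nodes.
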AnimBlueprint /BRCosmetics/Weapons/FORT_Melee/Pickaxe_HoppityHeist/Meshes/SK_Pickaxe_HoppityHeist_AnimBlueprint
target_skeleton=&/SK_Pickaxe_HoppityHeist_Skeleton
graphs [
 AnimGraph {nodes [
   n1 = RigidBody(ComponentPose=Local To Component(LocalPose=Use cached pose 'PrePose'())) {
    bOverrideWorldGravity=true
    overridePhysicsAsset=&/SK_Pickaxe_HoppityHeist_Physics
    alphaBoolBlend {blendOption=Linear}
    componentAppliedLinearAccClamp=<3000 3000 3000>
    componentLinearAccScale=<0.35 0.35 0.35>
    overrideWorldGravity {z=-600}
   }
   n2 = "Transform (Modify) Bone - Bone: dyn_fnt_ribbon_02"(
    ComponentPose=Local To Component(LocalPose=Use cached pose 'PrePose'())
    Alpha=1
    Translation="-0.090,-0.364,-0.000"
    Rotation="0.000,-13.774,0.000"
    Scale="1,1,1"
   ) {
    rotationMode=BMM_Additive
    rotationSpace=BCS_BoneSpace
    translationMode=BMM_Additive
    translationSpace=BCS_BoneSpace
    alphaBoolBlend {blendOption=Linear}
    boneToModify {boneName=dyn_fnt_ribbon_02}
    rotation {yaw=-13.7745}
    translation=<-0.09 -0.3639 0>
   }
   n3 = "Transform (Modify) Bone - Bone: dyn_fnt_ribbon_03"(ComponentPose=@n2, Alpha=1, Translation="0,0,0", Rotation="0.000,-19.891,0.000", Scale="1,1,1") {
    rotationMode=BMM_Additive
    rotationSpace=BCS_BoneSpace
    translationMode=BMM_Additive
    translationSpace=BCS_BoneSpace
    alphaBoolBlend {blendOption=Linear}
    boneToModify {boneName=dyn_fnt_ribbon_03}
    rotation {yaw=-19.891}
   }
   n4 = "Transform (Modify) Bone - Bone: dyn_bck_ribbon_02"(ComponentPose=@n3, Alpha=1, Translation="-0.027,-0.105,0.000", Rotation="3.322,-14.359,0.000", Scale="1,1,1") {
    rotationMode=BMM_Additive
    rotationSpace=BCS_BoneSpace
    translationMode=BMM_Additive
    translationSpace=BCS_BoneSpace
    alphaBoolBlend {blendOption=Linear}
    boneToModify {boneName=dyn_bck_ribbon_02}
    rotation {pitch=3.322, yaw=-14.359}
    translation=<-0.0269 -0.1047 0>
   }
   n5 = "Transform (Modify) Bone - Bone: dyn_bck_ribbon_03"(ComponentPose=@n4, Alpha=1, Translation="-0.133,-0.320,0.000", Rotation="0.000,-22.478,0.000", Scale="1,1,1") {
    rotationMode=BMM_Additive
    rotationSpace=BCS_BoneSpace
    translationMode=BMM_Additive
    translationSpace=BCS_BoneSpace
    alphaBoolBlend {blendOption=Linear}
    boneToModify {boneName=dyn_bck_ribbon_03}
    rotation {yaw=-22.4779}
    translation=<-0.133 -0.32>
   }
   n6 = Component To Local(ComponentPose=@n5)
   n7 = "Transform (Modify) Bone - Bone: dyn_fnt_ribbon_01"(
    ComponentPose=Local To Component(LocalPose=Use cached pose 'PrePose'())
    Alpha=1
    Translation="2.803,1.930,0.076"
    Rotation="-56.444,-46.557,31.934"
    Scale="0.95,1.000,1.000"
   ) {
    rotationMode=BMM_Additive
    rotationSpace=BCS_BoneSpace
    scaleMode=BMM_Additive
    scaleSpace=BCS_BoneSpace
    translationMode=BMM_Additive
    translationSpace=BCS_BoneSpace
    alphaBoolBlend {blendOption=Linear}
    boneToModify {boneName=dyn_fnt_ribbon_01}
    rotation=<rot -56.444 -46.557 31.934>
    scale {x=0.95}
    translation=<2.803 1.93 0.076>
   }
   n8 = "Transform (Modify) Bone - Bone: dyn_bck_ribbon_01"(ComponentPose=@n7, Alpha=1, Translation="2.647,1.178,-0.036", Rotation="-51.900,-42.061,28.278", Scale="1,1,1") {
    rotationMode=BMM_Additive
    rotationSpace=BCS_BoneSpace
    translationMode=BMM_Additive
    translationSpace=BCS_BoneSpace
    alphaBoolBlend {blendOption=Linear}
    boneToModify {boneName=dyn_bck_ribbon_01}
    rotation=<rot -51.9 -42.061 28.278>
    translation=<2.647 1.178 -0.036>
   }
   n9 = Component To Local(ComponentPose=@n8)
   n10 = Get bIsInFrontEnd()
   n11 = Trail controller - Bone: dyn_fnt_ribbon_05(
    ComponentPose=Local To Component(LocalPose=Blend Poses by bool(BlendPose_0=@n6, BlendPose_1=@n9, bActiveValue=@n10) {blendType=Linear})
    Alpha=1
   ) {
    chainLength=5
    relaxationSpeedScale=2
    rotationLimits [{limitMax=<180 180 180>}, {limitMax=<180 180 180>}, {limitMax=<180 180 180>}, {limitMax=<180 180 180>}, {limitMax=<180 180 180>}]
    rotationOffsets [<0 0 0>, <0 0 0>, <0 0 0>, <0 0 0>, <0 0 0>]
    trailBone {boneName=dyn_fnt_ribbon_05}
   }
   n12 = Trail controller - Bone: dyn_bck_ribbon_05(ComponentPose=@n11, Alpha=1) {
    chainLength=4
    relaxationSpeedScale=3
    rotationLimits [{limitMax=<180 180 180>}, {limitMax=<180 180 180>}, {limitMax=<180 180 180>}, {limitMax=<180 180 180>}]
    rotationOffsets [<0 0 0>, <0 0 0>, <0 0 0>, <0 0 0>]
    trailBone {boneName=dyn_bck_ribbon_05}
   }
   Output Pose(Result=Blend Poses by bool(BlendPose_0=Component To Local(ComponentPose=@n1), BlendPose_1=Component To Local(ComponentPose=@n12), bActiveValue=Get bRigidBodyEnabled()) {blendType=Linear}) {layerGroup=None}
   AnimGraphNode_SubInput_0 = Input Pose()
   Save cached pose 'PrePose'(Pose=@AnimGraphNode_SubInput_0) {cacheName=PrePose}
  ]}
 EventGraph {nodes [
   TryGetPawnOwner()
   Event BlueprintUpdateAnimation()
  ]}
]
variables []
